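# Niagara swing trail for the RoyalAngst pickaxe cosmetic, drawn as two short-lived gold ribbons.
# Both emitters use AnimTrail modules and a system socket reader to follow separate socket pairs; the first narrows over particle life, the second holds its width.
# A VariantColorSelector pinned at zero supplies gold over a dark variant, with time-of-day tint and dimming.
# Alpha fades over both particle age and emitter loop age.
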NiagaraSystem /BRCosmetics/Weapons/FORT_Melee/Pickaxe_RoyalAngst/FX/NS_Pickaxe_RoyalAngst_Trail
dependencies {
 usedDataInterfaces [&NiagaraDataInterfaceCurve, &NiagaraDataInterfaceSocketReader]
 usedDynamicInputs []
 usedModules [
  &FN_VariantManager
  &self:SystemSpawnScript.SystemScriptSource.SystemScriptGraph.NiagaraNodeAssignment_0.SetVariables_#1
  &self:SystemSpawnScript.SystemScriptSource.SystemScriptGraph.NiagaraNodeAssignment_1.SetVariables_#2
  &SystemState
 ]
}
emitters {
 Trail_Gold_01 {
  emitterUpdateScript [
   &EmitterState(Life Cycle Mode=::NewEnumerator1, Loop Behavior=::NewEnumerator1, Loop Delay=0.05, Loop Duration=0.5, UseLoopDelay=false)
   &FN_SpawnAnimTrail_v2(AnimTrailWidthMode=::FromSecond)
  ]
  particleSpawnScript [
   /Game/Effects/Niagara/Modules/AnimTrail/FN_InitializeAnimTrail/FN_InitializeAnimTrailBoneIndecies(Second Bone or Socket Index=1)
   /Game/Effects/Niagara/Modules/AnimTrail/FN_InitializeAnimTrail/FN_InitializeAnimTrail_SocketReader/FN_InitializeAnimTrail_SocketReader(
    Color:LinearColor=&FN_Select_Color_from_Time_Of_Day(
     Morning=$System.Color Main
     Day=$System.Color Main
     Evening=$System.Color Main
     Night=&MultiplyLinearColor(
      Linear Color=$System.Color Main
      Scale Factor=&MakeVector4_FromLinearColor(LinearColor=<rgba 0.3652 0.5529 1 1>)
     )
    )
    Lifetime=0.22
    Socket Reader:NiagaraDataInterfaceSocketReader=$System.Socket Reader
   ) {enabled=false}
   &FN_InitializeAnimTrail_SocketReader(Lifetime=0.22, Socket Reader:NiagaraDataInterfaceSocketReader=$System.Socket Reader) {module=FN_InitializeAnimTrail_SocketReader001}
  ]
  particleUpdateScript [
   /Game/Effects/Niagara/Modules/AnimTrail/FN_UpdateAnimTrail/FN_UpdateAnimTrail_SocketReader/FN_UpdateAnimTrail_SocketReader(Socket Reader:NiagaraDataInterfaceSocketReader=$System.Socket Reader)
   &ScaleRibbonWidth(Ribbon Width Scale=&FloatFromCurve(FloatCurve=DataInterface() {propertyValues {Curve {keys [
         {arriveTangent=-1.2579, interpMode=RCIM_Cubic, time=0, value=1}
         {arriveTangent=-0.9664, interpMode=RCIM_Cubic, leaveTangent=-0.9664, tangentMode=RCTM_User, time=1, value=0}
        ]}}}))
   &ScaleColor(
    Scale Mode=::NewEnumerator1
    Scale RGBA=&MakeVector4_FromLinearColor(LinearColor=&VariantColorSelector(
      Variant Amount=0
      Morning Color=<rgba 1 1 1 1>
      Day Color=<rgba 1 1 1 1>
      Evening Color=<rgba 1 1 1 1>
      Night Color=<rgba 0.3652 0.5529 1 1>
      Use TOD?=true
      Original Color=<rgba 0.702 0.4392 0.0235 1>
      Variant 1 Color=<rgba 0.0319 0.0222 0.0232 1>
      Variant 2 Color=?
     ))
    ScaleRGBA=true
   ) {module=ScaleColor001}
   &ScaleColor(
    Scale Alpha=&Multiply_Float(
     A=&FloatFromCurve(FloatCurve=DataInterface() {propertyValues {Curve {keys [
          {interpMode=RCIM_Cubic, leaveTangent=-2.5656, tangentMode=RCTM_Break, time=0, value=1}
          {arriveTangent=-0.151, tangentMode=RCTM_User, time=0.5, value=0}
         ]}}})
     B=&FNTimeOfDayScalar(Night Scale=0.5, Morning Scale=0.6, Evening Scale=0.6)
    )
    Scale RGB=&Multiply_Vector(
     A=&FN_Select_Vector_from_Time_Of_Day(Morning=<1 1 1>, Day=<1 1 1>, Evening=<1 1 1>, Night=<1 1 1>)
     B=&VectorFromFloat(Value=&FNTimeOfDayScalar(Night Scale=1, Morning Scale=0.6, Evening Scale=0.6))
    )
    ScaleA=true
    ScaleRGB=true
   ) {module=ScaleColor002}
   &ScaleColor(
    Scale Alpha=&FloatFromCurve(
     FloatCurve=DataInterface() {propertyValues {Curve {keys [{interpMode=RCIM_Cubic, tangentMode=RCTM_User, time=0, value=1}, {interpMode=RCIM_Cubic, tangentMode=RCTM_User, time=1, value=0}]}}}
     CurveIndex=$Emitter.NormalizedLoopAge
     Scale Curve=&Multiply_Float(A=0.8, B=&FN_Select_Float_from_Time_Of_Day(Morning=1, Day=1, Evening=1, Night=1))
    )
    ScaleA=true
    ScaleRGB=true
   )
  ]
  rendererClasses [&NiagaraRibbonRendererProperties]
  renderers [{rendererClass=&NiagaraRibbonRendererProperties}]
 }
 Trail_Gold_02 {
  emitterUpdateScript [
   &EmitterState(Life Cycle Mode=::NewEnumerator1, Loop Behavior=::NewEnumerator1, Loop Delay=0.05, Loop Duration=0.5, UseLoopDelay=false)
   &FN_SpawnAnimTrail_v2(AnimTrailWidthMode=::FromSecond)
  ]
  particleSpawnScript [
   &FN_InitializeAnimTrailBoneIndecies(First Bone or Socket Index=2, Second Bone or Socket Index=3)
   &FN_InitializeAnimTrail_SocketReader(
    AnimTrailWidth=2
    Color:LinearColor=&FN_Select_Color_from_Time_Of_Day(
     Morning=$System.Color Main
     Day=$System.Color Main
     Evening=$System.Color Main
     Night=&MultiplyLinearColor(
      Linear Color=$System.Color Main
      Scale Factor=&MakeVector4_FromLinearColor(LinearColor=<rgba 0.3652 0.5529 1 1>)
     )
    )
    Lifetime=0.22
    Socket Reader:NiagaraDataInterfaceSocketReader=$System.Socket Reader
   ) {enabled=false}
   &FN_InitializeAnimTrail_SocketReader(AnimTrailWidth=2, Lifetime=0.22, Socket Reader:NiagaraDataInterfaceSocketReader=$System.Socket Reader) {module=FN_InitializeAnimTrail_SocketReader001}
  ]
  particleUpdateScript [
   &FN_UpdateAnimTrail_SocketReader(Socket Reader:NiagaraDataInterfaceSocketReader=$System.Socket Reader)
   &ScaleRibbonWidth(Ribbon Width Scale=&FloatFromCurve(FloatCurve=DataInterface() {propertyValues {Curve {keys [{arriveTangent=-1.2579, interpMode=RCIM_Cubic, time=0, value=1}, <at 1 1 RCIM_Cubic>]}}}))
   &ScaleColor(
    Scale Mode=::NewEnumerator1
    Scale RGBA=&MakeVector4_FromLinearColor(LinearColor=&VariantColorSelector(
      Variant Amount=0
      Morning Color=<rgba 1 1 1 1>
      Day Color=<rgba 1 1 1 1>
      Evening Color=<rgba 1 1 1 1>
      Night Color=<rgba 0.3652 0.5529 1 1>
      Use TOD?=true
      Original Color=<rgba 0.702 0.4392 0.0235 1>
      Variant 1 Color=<rgba 0.0319 0.0222 0.0232 1>
      Variant 2 Color=?
     ))
    ScaleRGBA=true
   ) {module=ScaleColor001}
   &ScaleColor(
    Scale Alpha=&Multiply_Float(
     A=&FloatFromCurve(
      FloatCurve=DataInterface() {propertyValues {Curve {keys [
          {interpMode=RCIM_Cubic, leaveTangent=-2.5656, tangentMode=RCTM_Break, time=0, value=1}
          {arriveTangent=-0.151, tangentMode=RCTM_User, time=0.5, value=0}
         ]}}}
      Scale Curve=0.8
     )
     B=&FNTimeOfDayScalar(Night Scale=0.5, Morning Scale=0.6, Evening Scale=0.6)
    )
    Scale RGB=&Multiply_Vector(
     A=&FN_Select_Vector_from_Time_Of_Day(Morning=<1 1 1>, Day=<1 1 1>, Evening=<1 1 1>, Night=<1 1 1>)
     B=&VectorFromFloat(Value=&FNTimeOfDayScalar(Night Scale=1, Morning Scale=0.6, Evening Scale=0.6))
    )
    ScaleA=true
    ScaleRGB=true
   ) {module=ScaleColor002}
   &ScaleColor(
    Scale Alpha=&FloatFromCurve(
     FloatCurve=DataInterface() {propertyValues {Curve {keys [{interpMode=RCIM_Cubic, tangentMode=RCTM_User, time=0, value=1}, {interpMode=RCIM_Cubic, tangentMode=RCTM_User, time=1, value=0}]}}}
     CurveIndex=$Emitter.NormalizedLoopAge
     Scale Curve=&Multiply_Float(A=0.8, B=&FN_Select_Float_from_Time_Of_Day(Morning=1, Day=1, Evening=1, Night=1))
    )
    ScaleA=true
    ScaleRGB=true
   )
  ]
  rendererClasses [&NiagaraRibbonRendererProperties]
  renderers [{rendererClass=&NiagaraRibbonRendererProperties}]
 }
}
summary {
 systemName=NS_Pickaxe_RoyalAngst_Trail
 emitters [
  {emitterName=Trail_Gold_01, rendererClasses [&NiagaraRibbonRendererProperties]}
  {emitterName=Trail_Gold_02, rendererClasses [&NiagaraRibbonRendererProperties]}
 ]
 userVariables [
  User.bIsFrontEnd {type=NiagaraBool}
  User.bIsFrontEndPreview {type=NiagaraBool}
  User.RibbonIDOffset {type=NiagaraInt32}
  User.VariantID {type=NiagaraFloat, defaultValue=0, description=""}
 ]
}
