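# Physics asset for the RoyalAngst two-handed pickaxe, providing collision bodies and joints for two dangling chains on the weapon.
# A long capsule covers the handle, and each chain has a base body with free angular motion on the handle, followed by eight small capsule links.
# The links are jointed in sequence with zero twist limits and bounded swing, tighter at each chain's first link.
# Chain B also carries a ring body branching from its fifth link, limited to one swing axis.
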PhysicsAsset /BRCosmetics/Weapons/FORT_Melee/Pickaxe_RoyalAngst/Meshes/RoyalAngst_Axe_PhysicsAsset
bodies {
 Melee_TwoHanded_Handle [{shapeType=Capsule, length=70, radius=2.2, center=<0 0 13.8279>, rotation=<rot 0 -90 0>}]
 chainA_01_dyn [{shapeType=Capsule, length=1, radius=0.9, shapeName=chainA_01_dyn_capsule, center=<1.2 0 0>, rotation=<rot 90 0 0>}]
 chainA_02_dyn [{shapeType=Capsule, length=1, radius=0.9, shapeName=chainA_02_dyn_capsule, center=<1.2 0 0>, rotation=<rot 90 0 0>}]
 chainA_03_dyn [{shapeType=Capsule, length=1, radius=0.9, shapeName=chainA_03_dyn_capsule, center=<1.2 0 0>, rotation=<rot 90 0 0>}]
 chainA_04_dyn [{shapeType=Capsule, length=1, radius=0.9, shapeName=chainA_04_dyn_capsule, center=<1.2 0 0>, rotation=<rot 90 0 0>}]
 chainA_05_dyn [{shapeType=Capsule, length=1, radius=0.8, shapeName=chainA_05_dyn_capsule, center=<1.2 0 0>, rotation=<rot 90 0 0>}]
 chainA_06_dyn [{shapeType=Capsule, length=1, radius=0.8, shapeName=chainA_06_dyn_capsule, center=<1.2 0 0>, rotation=<rot 90 0 0>}]
 chainA_07_dyn [{shapeType=Capsule, length=1, radius=0.7, shapeName=chainA_07_dyn_capsule, center=<1.2 0 0>, rotation=<rot 90 0 0>}]
 chainA_08_dyn [{shapeType=Capsule, length=1, radius=0.7, shapeName=chainA_08_dyn_capsule, center=<1.2 0 0>, rotation=<rot 90 0 0>}]
 chainA_base [{shapeType=Capsule, length=0.505, radius=0.505, shapeName=chainA_base_capsule, rotation=<rot 0 -90 0>}]
 chainB_01_dyn [{shapeType=Capsule, length=1, radius=0.8, shapeName=chainB_01_dyn_capsule, center=<1.1 0 0>, rotation=<rot 90 0 0>}]
 chainB_02_dyn [{shapeType=Capsule, length=1, radius=0.9, shapeName=chainB_02_dyn_capsule, center=<1.2 0 0>, rotation=<rot 90 0 0>}]
 chainB_03_dyn [{shapeType=Capsule, length=1, radius=0.9, shapeName=chainB_03_dyn_capsule, center=<1.2 0 0>, rotation=<rot 90 0 0>}]
 chainB_04_dyn [{shapeType=Capsule, length=1, radius=0.9, shapeName=chainB_04_dyn_capsule, center=<1.2 0 0>, rotation=<rot 90 0 0>}]
 chainB_05_dyn [{shapeType=Capsule, length=1, radius=0.9, shapeName=chainB_05_dyn_capsule, center=<1.2 0 0>, rotation=<rot 90 0 0>}]
 chainB_06_dyn [{shapeType=Capsule, length=1, radius=0.75, shapeName=chainB_06_dyn_capsule, center=<1.2 0 0>, rotation=<rot 90 0 0>}]
 chainB_07_dyn [{shapeType=Capsule, length=1, radius=0.75, shapeName=chainB_07_dyn_capsule, center=<1.2 0 0>, rotation=<rot 90 0 0>}]
 chainB_08_dyn [{shapeType=Capsule, length=1, radius=0.75, shapeName=chainB_08_dyn_capsule, center=<1.2 0 0>, rotation=<rot 90 0 0>}]
 chainB_base [{shapeType=Capsule, length=0.505, radius=0.505, shapeName=chainB_base_capsule, rotation=<rot 0 -90 0>}]
 chainB_ring_dyn [{shapeType=Capsule, length=4.5, radius=0.5, shapeName=chainB_ring_dyn_capsule, center=<3 0 0>, rotation=<rot 90 0 0>}]
}
constraints [
 {
  bone1Name=chainA_base
  bone2Name=Melee_TwoHanded_Handle
  swing1LimitDegrees=45
  swing1Motion=Free
  swing2LimitDegrees=45
  swing2Motion=Free
  twistLimitDegrees=45
  twistMotion=Free
 }
 {bone1Name=chainA_01_dyn, bone2Name=chainA_base, swing1LimitDegrees=30, swing2LimitDegrees=0, twistLimitDegrees=0}
 {bone1Name=chainA_02_dyn, bone2Name=chainA_01_dyn, swing1LimitDegrees=45, swing2LimitDegrees=45, twistLimitDegrees=0}
 {bone1Name=chainA_03_dyn, bone2Name=chainA_02_dyn, swing1LimitDegrees=45, swing2LimitDegrees=45, twistLimitDegrees=0}
 {bone1Name=chainA_04_dyn, bone2Name=chainA_03_dyn, swing1LimitDegrees=45, swing2LimitDegrees=45, twistLimitDegrees=0}
 {bone1Name=chainA_05_dyn, bone2Name=chainA_04_dyn, swing1LimitDegrees=45, swing2LimitDegrees=45, twistLimitDegrees=0}
 {bone1Name=chainA_06_dyn, bone2Name=chainA_05_dyn, swing1LimitDegrees=45, swing2LimitDegrees=45, twistLimitDegrees=0}
 {bone1Name=chainA_07_dyn, bone2Name=chainA_06_dyn, swing1LimitDegrees=45, swing2LimitDegrees=45, twistLimitDegrees=0}
 {bone1Name=chainA_08_dyn, bone2Name=chainA_07_dyn, swing1LimitDegrees=45, swing2LimitDegrees=45, twistLimitDegrees=0}
 {
  bone1Name=chainB_base
  bone2Name=Melee_TwoHanded_Handle
  swing1LimitDegrees=45
  swing1Motion=Free
  swing2LimitDegrees=45
  swing2Motion=Free
  twistLimitDegrees=45
  twistMotion=Free
 }
 {bone1Name=chainB_01_dyn, bone2Name=chainB_base, swing1LimitDegrees=25, swing2LimitDegrees=5, twistLimitDegrees=0}
 {bone1Name=chainB_02_dyn, bone2Name=chainB_01_dyn, swing1LimitDegrees=40, swing2LimitDegrees=40, twistLimitDegrees=0}
 {bone1Name=chainB_03_dyn, bone2Name=chainB_02_dyn, swing1LimitDegrees=40, swing2LimitDegrees=40, twistLimitDegrees=0}
 {bone1Name=chainB_04_dyn, bone2Name=chainB_03_dyn, swing1LimitDegrees=40, swing2LimitDegrees=40, twistLimitDegrees=0}
 {bone1Name=chainB_05_dyn, bone2Name=chainB_04_dyn, swing1LimitDegrees=40, swing2LimitDegrees=40, twistLimitDegrees=0}
 {bone1Name=chainB_06_dyn, bone2Name=chainB_05_dyn, swing1LimitDegrees=40, swing2LimitDegrees=40, twistLimitDegrees=0}
 {bone1Name=chainB_07_dyn, bone2Name=chainB_06_dyn, swing1LimitDegrees=45, swing2LimitDegrees=45, twistLimitDegrees=0}
 {bone1Name=chainB_08_dyn, bone2Name=chainB_07_dyn, swing1LimitDegrees=45, swing2LimitDegrees=45, twistLimitDegrees=0}
 {bone1Name=chainB_ring_dyn, bone2Name=chainB_05_dyn, swing1LimitDegrees=0, swing2LimitDegrees=25, twistLimitDegrees=0}
]
