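# Animation Blueprint for the Royal Angst pickaxe that drives secondary motion on its two dangling chains.
# In the front end, a cached pose with a nudged handle bone feeds either a RigidBody simulation or, when rigid bodies are disabled, bone offsets and two Trail controllers.
# Outside the front end, a zero-gravity RigidBody simulation feeds additive chain rotations.
# The event graph updates GravityOverride through a deprecated helper and sets an Is RBAN flag.
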AnimBlueprint /BRCosmetics/Weapons/FORT_Melee/Pickaxe_RoyalAngst/Meshes/RoyalAngst_Axe_AnimBP
target_skeleton=&/RoyalAngst_Axe_Skeleton
graphs [
 AnimGraph {nodes [
   n1 = "Transform (Modify) Bone - Bone: chainB_02_dyn"(ComponentPose=Local To Component(LocalPose=Use cached pose 'FrontEndFix'()), Alpha=1, Translation="0,0,0", Rotation="0,-90,0", Scale="1,1,1") {rotationMode=BMM_Additive, alphaBoolBlend {blendOption=Linear}, boneToModify {boneName=chainB_02_dyn}, rotation {yaw=-90}}
   n2 = "Transform (Modify) Bone - Bone: chainB_ring_dyn"(ComponentPose=@n1, Alpha=1, Translation="0,0,0", Rotation="-60,0,0", Scale="1,1,1") {rotationMode=BMM_Additive, alphaBoolBlend {blendOption=Linear}, boneToModify {boneName=chainB_ring_dyn}, rotation {pitch=-60}}
   n3 = "Transform (Modify) Bone - Bone: chainA_02_dyn"(ComponentPose=@n2, Alpha=1, Translation="0,0,0", Rotation="0,90,0", Scale="1,1,1") {rotationMode=BMM_Additive, alphaBoolBlend {blendOption=Linear}, boneToModify {boneName=chainA_02_dyn}, rotation {yaw=90}}
   n4 = Trail controller - Bone: chainA_01_dyn!(ComponentPose=@n3, Alpha=0.25, RelaxationSpeedScale:float=2, FakeVelocity:Vector="0,0,0") -> (Pose) {
    bActorSpaceFakeVel=true
    bShowTrailLocation=true
    bUsePlanarLimit=true
    chainLength=7
    debugLifeTime=0.1
    maxDeltaTime=0.0333
    relaxationSpeedScale=1.2
    alphaScaleBiasClamp {interpSpeedDecreasing=25, interpSpeedIncreasing=25}
    baseJoint {boneName=b_AR_Root}
    rotationLimits [
     {limitMax=<180 180 180>}
     {limitMax=<180 180 180>}
     {limitMax=<180 180 180>}
     {limitMax=<180 180 180>}
     {limitMax=<180 180 180>}
     {limitMax=<180 180 180>}
     {limitMax=<180 180 180>}
    ]
    rotationOffsets [<0 0 0>, <0 0 0>, <0 0 0>, <0 0 0>, <0 0 0>, <0 0 0>, <0 0 0>]
    trailBone {boneName=chainA_01_dyn}
    trailRelaxationSpeed {editorCurveData {keys [<at 0 15 RCIM_Linear>,<at 1 6 RCIM_Linear>]}}
   }
   n5 = Trail controller - Bone: chainB_01_dyn!(ComponentPose=@n4, Alpha=0.25, RelaxationSpeedScale:float=2, FakeVelocity:Vector="0,0,0") -> (Pose) {
    bActorSpaceFakeVel=true
    bShowTrailLocation=true
    bUsePlanarLimit=true
    chainLength=7
    debugLifeTime=0.1
    maxDeltaTime=0.0333
    relaxationSpeedScale=1.2
    alphaScaleBiasClamp {interpSpeedDecreasing=25, interpSpeedIncreasing=25}
    baseJoint {boneName=b_AR_Root}
    rotationLimits [
     {limitMax=<180 180 180>}
     {limitMax=<180 180 180>}
     {limitMax=<180 180 180>}
     {limitMax=<180 180 180>}
     {limitMax=<180 180 180>}
     {limitMax=<180 180 180>}
     {limitMax=<180 180 180>}
    ]
    rotationOffsets [<0 0 0>, <0 0 0>, <0 0 0>, <0 0 0>, <0 0 0>, <0 0 0>, <0 0 0>]
    trailBone {boneName=chainB_01_dyn}
    trailRelaxationSpeed {editorCurveData {keys [<at 0 15 RCIM_Linear>,<at 1 6 RCIM_Linear>]}}
   }
   n6 = Component To Local(ComponentPose=@n5)
   n7 = RigidBody!(ComponentPose=Local To Component(LocalPose=Use cached pose 'FrontEndFix'()), Alpha, OverrideWorldGravity:Vector="0,0,-980", ExternalForce) -> (Pose) {
    bOverrideWorldGravity=true
    overridePhysicsAsset=&/RoyalAngst_Axe_PhysicsAsset
    simulationSpace=BaseBoneSpace
    alphaBoolBlend {blendOption=Linear}
    baseBoneRef {boneName=b_AR_Root}
    componentAppliedLinearAccClamp=<3000 3000 3000>
    componentLinearAccScale=<0.3 0.4 0.5>
    componentLinearVelScale=<0.3 0.4 0.5>
    simSpaceSettings {
     maxAngularAcceleration=1000
     maxAngularVelocity=3
     maxLinearAcceleration=700
     maxLinearVelocity=500
     velocityScaleZ=0
     worldAlpha=0.75
     externalLinearDragV=<0.2 0.2>
    }
   }
   n8 = Component To Local(ComponentPose=@n7)
   n9 = Get bRigidBodyEnabled()
   n10 = RigidBody(ComponentPose=Local To Component(LocalPose=Use cached pose 'StartPose'()), OverrideWorldGravity="0,0,0") {
    bOverrideWorldGravity=true
    overridePhysicsAsset=&RoyalAngst_Axe_PhysicsAsset
    simulationSpace=BaseBoneSpace
    alphaBoolBlend {blendOption=Linear}
    baseBoneRef {boneName=b_AR_Root}
    componentAppliedLinearAccClamp=<3000 3000 3000>
    componentLinearAccScale=<0.3 0.3 0.4>
    componentLinearVelScale=<0.3 0.3 0.4>
    simSpaceSettings {
     maxAngularAcceleration=1000
     maxAngularVelocity=3
     maxLinearAcceleration=700
     maxLinearVelocity=500
     velocityScaleZ=0
     worldAlpha=0.45
     externalLinearDragV=<0.2 0.2>
    }
   }
   n11 = "Transform (Modify) Bone - Bone: chainB_02_dyn"(ComponentPose=@n10, Alpha=1, Translation="0,0,0", Rotation="0,-90,0", Scale="1,1,1") {rotationMode=BMM_Additive, alphaBoolBlend {blendOption=Linear}, boneToModify {boneName=chainB_02_dyn}, rotation {yaw=-90}}
   n12 = "Transform (Modify) Bone - Bone: chainB_ring_dyn"(ComponentPose=@n11, Alpha=1, Translation="0,0,0", Rotation="-60,0,0", Scale="1,1,1") {rotationMode=BMM_Additive, alphaBoolBlend {blendOption=Linear}, boneToModify {boneName=chainB_ring_dyn}, rotation {pitch=-60}}
   n13 = "Transform (Modify) Bone - Bone: chainA_02_dyn"(ComponentPose=@n12, Alpha=1, Translation="0,0,0", Rotation="0,90,0", Scale="1,1,1") {rotationMode=BMM_Additive, alphaBoolBlend {blendOption=Linear}, boneToModify {boneName=chainA_02_dyn}, rotation {yaw=90}}
   Output Pose(Result=Blend Poses by bool(
     BlendPose_0=Blend Poses by bool(BlendPose_0=@n8, BlendPose_1=@n6, bActiveValue=@n9) {blendType=Linear}
     BlendPose_1=Component To Local(ComponentPose=@n13)
     bActiveValue=Get bIsInFrontEnd()
    ) {blendType=Linear})
   Save cached pose 'StartPose'(Pose=Input Pose() {name=InPose_0}) {cacheName=StartPose}
   n14 = "Transform (Modify) Bone - Bone: Melee_TwoHanded_Handle"!(
    ComponentPose=Local To Component(LocalPose=Use cached pose 'StartPose'())
    bAlphaBoolEnabled:bool=True
    Translation="0,-0.3,0"
    Rotation="0,0,0"
    Scale="1,1,1"
   ) -> (Pose) {
    alphaInputType=Bool
    rotationMode=BMM_Additive
    translationMode=BMM_Additive
    alphaBoolBlend {blendOption=Linear}
    boneToModify {boneName=Melee_TwoHanded_Handle}
    translation {y=-0.3}
   }
   Save cached pose 'FrontEndFix'(Pose=Component To Local(ComponentPose=@n14)) {cacheName=FrontEndFix}
  ]}
 EventGraph {nodes [
   Event BlueprintUpdateAnimation()
   > n1 = Gravity Override State Function DEPRECATED(CharacterPartAnimInstance=Self-Reference(), GameTimeInSeconds=GetGameTimeInSeconds(), GravityOverrideParams=Get Gravity Override Params())
   > Set GravityOverride(GravityOverride=@n1)
   > n2 = Branch(Condition=Get bRigidBodyEnabled())
   > Set Is RBAN(Is RBAN:double=0) -> (Output_Get:double)
   Set Is RBAN(execute=@n2.else, Is RBAN=1)
   TryGetPawnOwner()
  ]}
]
variables [
 Gravity Override Params {type=GravityOverrideParamsStruct}
 GravityOverride {type=Vector}
 Is RBAN {type=number}
]
